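# Conversation database for the Cuddle Team Leader NPC, giving its greeting dialogue and access to shared NPC conversation branches.
# The entry point feeds a speech node with a default coaster-themed line plus contextual variants for a first-ever conversation or for players wearing specific character skins.
# The first-meeting line takes its priority from a curve row, while the skin lines clear theirs.
# The speech node then branches through link nodes to the shared quest and basic services conversations.
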ConversationDatabase /NPCLibrary/NPCs/CuddleTeamLeader/CuddleTeamLeader/NPCLibrary_Conversation_CuddleTeamLeader
entry_points [{tag=Conversation.Entry.CuddleTeamLeader, destinations [#1]}]
nodes [
 ConversationEntryPointNode_7 {
  id=#1
  class=ConversationEntryPointNode
  outputs [#2]
  properties {entryTag {tagName=Conversation.Entry.CuddleTeamLeader}}
 }
 Speech_C_4 {
  id=#2
  class=Speech_C
  outputs [#3, #4]
  properties {
   nodeName=""
   generalConfig {
    defaultMessage=I'm not accepting coaster safety complaints at this time.
    contextualMessages [
     {
      message=Have you tried my new coaster? Made the seatbelts myself.
      contextRequirements [{requirement {class=FortControllerRequirement_IsFirstEverConversationWithNPC}}]
      priority {curve {rowName=Tandem.Speech.Priority.FirstEverInteraction}}
     }
     {
      message="The coaster is ride at your own risk, sire."
      contextRequirements [{requirement {
         class=FortControllerRequirement_HasCID
         properties {allowedCIds [/BRCosmetics/Athena/Items/Cosmetics/Characters/CID_A_424_Athena_Commando_M_Lancelot]}
        }}]
      priority {value=0, curve {curveTable=null, rowName=None}}
     }
     {
      message="It takes guts to leave the life you had behind. "
      contextRequirements [{requirement {
         class=FortControllerRequirement_HasCID
         properties {allowedCIds [/BRCosmetics/Athena/Items/Cosmetics/Characters/CID_A_428_Athena_Commando_F_PinkWidow]}
        }}]
      priority {value=0, curve {curveTable=null, rowName=None}}
     }
     {
      message=Interested in sub-licensing your dragon friend for a theme park attraction?
      contextRequirements [{requirement {
         class=FortControllerRequirement_HasCID
         properties {allowedCIds [/BRCosmetics/Athena/Items/Cosmetics/Characters/CID_A_427_Athena_Commando_F_Fuchsia]}
        }}]
      priority {value=0, curve {curveTable=null, rowName=None}}
     }
     {
      message="If the coaster did that to you, I accept no liability."
      contextRequirements [{requirement {
         class=FortControllerRequirement_HasCID
         properties {allowedCIds [/BRCosmetics/Athena/Items/Cosmetics/Characters/CID_A_429_Athena_Commando_M_Collectable]}
        }}]
      priority {value=0, curve {curveTable=null, rowName=None}}
     }
     {
      message=The Syndicate's problem-solving style is... Interesting.
      contextRequirements [{requirement {
         class=FortControllerRequirement_HasCID
         properties {allowedCIds [&CID_A_425_Athena_Commando_F_BlueJay]}
        }}]
      priority {value=0, curve {curveTable=null, rowName=None}}
     }
     {
      message="You got a ticket? "
      contextRequirements [{requirement {
         class=FortControllerRequirement_HasCID
         properties {allowedCIds [/BRCosmetics/Athena/Items/Cosmetics/Characters/CID_A_423_Athena_Commando_M_Canary]}
        }}]
      priority {value=0, curve {curveTable=null, rowName=None}}
     }
    ]
   }
  }
 }
 ConversationLinkNode_0 {id=#4, class=ConversationLinkNode, properties {remoteEntryTag {tagName=Conversation.Entry.BasicServices}}}
 ConversationLinkNode_1 {id=#3, class=ConversationLinkNode, properties {remoteEntryTag {tagName=Conversation.Entry.NPCLibrary.Quests}}}
]
